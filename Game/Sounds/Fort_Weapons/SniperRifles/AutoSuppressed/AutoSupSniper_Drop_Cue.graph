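# AutoSupSniper_Drop_Cue is the sound cue for dropping the suppressed automatic sniper rifle.
# A single wave player feeds AutoSupSniper_Drop_01 straight to the output, with no randomization, at lowered pitch and volume.
# It uses the shared weapon drop attenuation and the loot collect sound class, and its concurrency override blocks new instances once the limit is reached.
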SoundCue /Game/Sounds/Fort_Weapons/SniperRifles/AutoSuppressed/AutoSupSniper_Drop_Cue
root=WavePlayer_1
nodes [{id=WavePlayer_1, properties {soundWaveAssetPtr=&/AutoSupSniper_Drop_01}}]
properties {
 attenuationSettings=&Fort_Weapon_Drop_Att
 bOverrideConcurrency=true
 maxDistance=3000
 pitchMultiplier=0.9
 soundClassObject=&Fort_Loot_Collect
 volumeMultiplier=0.6031
 concurrencyOverrides {resolutionRule=PreventNew}
}
