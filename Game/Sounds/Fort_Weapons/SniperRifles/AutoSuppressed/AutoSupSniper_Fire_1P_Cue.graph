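# First-person fire sound cue for the suppressed automatic sniper rifle.
# A Random node picks one of four equally weighted close-range shot waves, then a Modulator feeds the output.
# The cue slightly lowers pitch and volume, uses the first-person weapon fire sound class, and references a first-person projectile attenuation asset.
# It also references a first-person gunfire concurrency set.
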SoundCue /Game/Sounds/Fort_Weapons/SniperRifles/AutoSuppressed/AutoSupSniper_Fire_1P_Cue
root=Modulator_17
nodes [
 {id=Modulator_17, children [Random_1]}
 {id=Random_1, children [WavePlayer_1, WavePlayer_8, WavePlayer_9, WavePlayer_10], properties {weights [1,1,1,1]}}
 {id=WavePlayer_1, properties {soundWaveAssetPtr=&/AutoSupSniper_Fire_Close_01}}
 {id=WavePlayer_10, properties {soundWaveAssetPtr=&/AutoSupSniper_Fire_Close_04}}
 {id=WavePlayer_8, properties {soundWaveAssetPtr=&/AutoSupSniper_Fire_Close_02}}
 {id=WavePlayer_9, properties {soundWaveAssetPtr=&/AutoSupSniper_Fire_Close_03}}
]
properties {
 attenuationSettings=&Fort_Weapon_Projectile_FirstPerson_Athena_Att
 maxDistance=2097152
 pitchMultiplier=0.9
 soundClassObject=&Fort_WeaponRanged_Fire_1P_PSM
 volumeMultiplier=0.8
 attenuationOverrides {distanceAlgorithm=NaturalSound, falloffDistance=4000, attenuationShapeExtents {x=1000}}
 concurrencyOverrides {bLimitToOwner=true, maxCount=3, resolutionRule=StopOldest}
 concurrencySet [&Weapon_Gunfire_FirstPerson_Conc]
}
