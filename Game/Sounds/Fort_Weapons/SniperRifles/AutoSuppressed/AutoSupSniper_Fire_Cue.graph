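# Sound cue for the auto suppressed sniper rifle's gunfire, layered by distance into close, medium and distant sounds.
# A top mixer sums the three branches, each with its own attenuation, sound class override and modulation.
# Close and distant branches pick random waves, while the medium branch plays all three waves together through a mixer and two stacked volume modulators.
# Local and global gunfire concurrency sets are assigned to the cue.
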SoundCue /Game/Sounds/Fort_Weapons/SniperRifles/AutoSuppressed/AutoSupSniper_Fire_Cue
root=Mixer_1
nodes [
 {
  id=Attenuation_11
  children [Modulator_17]
  properties {attenuationSettings=&Fort_Weapon_Projectile_Close_Athena_Att}
 }
 {
  id=Attenuation_12
  children [Modulator_18]
  properties {attenuationSettings=&Fort_Weapon_Projectile_Medium_Mono_Athena_Att}
 }
 {
  id=Attenuation_13
  children [Modulator_19]
  properties {attenuationSettings=&Fort_Weapon_Projectile_Distant_Mono_Athena_Att}
 }
 {id=Mixer_1, children [SoundClass_14, SoundClass_16, SoundClass_15], properties {inputVolume [0.9,0.6,0.3]}}
 {id=Mixer_2, children [WavePlayer_14, WavePlayer_15, WavePlayer_20], properties {inputVolume [1,1,1]}}
 {id=Modulator_1, children [Mixer_2], properties {volumeMax=0.65, volumeMin=0.6}}
 {id=Modulator_17, children [Random_0], properties {pitchMax=1.1, pitchMin=0.9}}
 {id=Modulator_18, children [Modulator_1], properties {volumeMax=2.1, volumeMin=2}}
 {id=Modulator_19, children [Random_5], properties {volumeMax=2.05, volumeMin=1.95}}
 {id=Random_0, children [WavePlayer_2, WavePlayer_3, WavePlayer_4, WavePlayer_13], properties {weights [1,1,1,1]}}
 {id=Random_5, children [WavePlayer_21, WavePlayer_22, WavePlayer_23], properties {weights [1,1,1]}}
 {
  id=SoundClass_14
  children [Attenuation_11]
  properties {soundClassOverride=&Fort_WeaponRanged_Fire_3P_PSM}
 }
 {
  id=SoundClass_15
  children [Attenuation_13]
  properties {soundClassOverride=&Fort_WeaponRanged_Fire_Distant}
 }
 {
  id=SoundClass_16
  children [Attenuation_12]
  properties {soundClassOverride=&Fort_WeaponRanged_Fire_Distant}
 }
 {id=WavePlayer_13, properties {soundWaveAssetPtr=&/AutoSupSniper_Fire_Close_03}}
 {id=WavePlayer_14, properties {soundWaveAssetPtr=&/AutoSupSniper_Fire_Medium_01}}
 {id=WavePlayer_15, properties {soundWaveAssetPtr=&/AutoSupSniper_Fire_Medium_02}}
 {id=WavePlayer_2, properties {soundWaveAssetPtr=&/AutoSupSniper_Fire_Close_04}}
 {id=WavePlayer_20, properties {soundWaveAssetPtr=&/AutoSupSniper_Fire_Medium_03}}
 {id=WavePlayer_21, properties {soundWaveAssetPtr=&/AutoSupSniper_Fire_Distant_01}}
 {id=WavePlayer_22, properties {soundWaveAssetPtr=&/AutoSupSniper_Fire_Distant_02}}
 {id=WavePlayer_23, properties {soundWaveAssetPtr=&/AutoSupSniper_Fire_Distant_03}}
 {id=WavePlayer_3, properties {soundWaveAssetPtr=&/AutoSupSniper_Fire_Close_01}}
 {id=WavePlayer_4, properties {soundWaveAssetPtr=&/AutoSupSniper_Fire_Close_02}}
]
properties {
 maxDistance=27500
 soundClassObject=&Fort_WeaponRanged_Fire_3P_PSM
 volumeMultiplier=0.9
 attenuationOverrides {distanceAlgorithm=NaturalSound, falloffDistance=4000, attenuationShapeExtents {x=1000}}
 concurrencyOverrides {bLimitToOwner=true, maxCount=3, resolutionRule=StopOldest}
 concurrencySet [&Weapon_Gunfire_Local_Conc, &Weapon_Gunfire_Global_Conc]
}
