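# Sound cue for the magazine insert step of reloading the auto suppressed sniper rifle.
# A top mixer plays the main insert wave immediately alongside a fixed-delay, half-volume layer.
# That delayed layer applies slight random upward pitch to a blend of a quiet insert sample from the drum shotgun folder and a PDW insert-slam sample.
# It uses weapon reload attenuation, the weapon foley sound class, and overridden concurrency.
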SoundCue /Game/Sounds/Fort_Weapons/SniperRifles/AutoSuppressed/AutoSupSniper_ReloadInsert_Cue
root=Mixer_0
nodes [
 {id=Delay_0, children [Modulator_0], properties {delayMax=0.37, delayMin=0.37}}
 {id=Mixer_0, children [WavePlayer_0,Delay_0], properties {inputVolume [1,0.5]}}
 {id=Mixer_1, children [WavePlayer_6,WavePlayer_5], properties {inputVolume [0.2,1]}}
 {id=Modulator_0, children [Mixer_1], properties {pitchMax=1.1, pitchMin=1}}
 {id=WavePlayer_0, properties {soundWaveAssetPtr=&/AutoSupSniper_ReloadInsert_01}}
 {id=WavePlayer_5, properties {soundWaveAssetPtr=/Game/Sounds/Fort_Weapons/AutoRifle/PDW/PDW_Reload_InsertSlam_01}}
 {id=WavePlayer_6, properties {soundWaveAssetPtr=/Game/Sounds/Fort_Weapons/Shotgun/Drum/AutoSupSniper_ReloadInsert}}
]
properties {
 attenuationSettings=&Fort_Weapon_Reload_Att
 bDebug=true
 bOverrideConcurrency=true
 maxDistance=2250
 soundClassObject=&Fort_WeaponFoley
 volumeMultiplier=0.576
}
