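# Sound cue for the start of the auto suppressed sniper rifle's reload, played as weapon foley with reload attenuation.
# A mixer layers the dedicated reload-start wave with two quieter, delayed layers borrowed from the military sniper.
# The first is a random throw sound after half a second, pitched up by a modulator.
# The second is a random impact sound just over a second in.
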SoundCue /Game/Sounds/Fort_Weapons/SniperRifles/AutoSuppressed/AutoSupSniper_ReloadStart_Cue
root=Mixer_0
nodes [
 {id=Delay_0, children [Modulator_0], properties {delayMax=0.5, delayMin=0.5}}
 {id=Delay_1, children [Random_1], properties {delayMax=1.1, delayMin=1.1}}
 {id=Mixer_0, children [WavePlayer_1,Delay_0,Delay_1], properties {inputVolume [1,0.5,0.7]}}
 {id=Modulator_0, children [Random_0], properties {pitchMax=1.3, pitchMin=1.2}}
 {id=Random_0, children [WavePlayer_2,WavePlayer_3,WavePlayer_4], properties {weights [1,1,1]}}
 {id=Random_1, children [WavePlayer_5,WavePlayer_6,WavePlayer_7,WavePlayer_8], properties {weights [1,1,1,1]}}
 {id=WavePlayer_1, properties {soundWaveAssetPtr=&/AutoSupSniper_ReloadStart_01}}
 {id=WavePlayer_2, properties {soundWaveAssetPtr=/Game/Sounds/Fort_Weapons/SniperRifles/MilitarySniper/MilitarySniper_Reload_Throw_01}}
 {id=WavePlayer_3, properties {soundWaveAssetPtr=/Game/Sounds/Fort_Weapons/SniperRifles/MilitarySniper/MilitarySniper_Reload_Throw_02}}
 {id=WavePlayer_4, properties {soundWaveAssetPtr=/Game/Sounds/Fort_Weapons/SniperRifles/MilitarySniper/MilitarySniper_Reload_Throw_03}}
 {id=WavePlayer_5, properties {soundWaveAssetPtr=/Game/Sounds/Fort_Weapons/SniperRifles/MilitarySniper/MilitarySniper_Reload_Impact_01}}
 {id=WavePlayer_6, properties {soundWaveAssetPtr=/Game/Sounds/Fort_Weapons/SniperRifles/MilitarySniper/MilitarySniper_Reload_Impact_02}}
 {id=WavePlayer_7, properties {soundWaveAssetPtr=/Game/Sounds/Fort_Weapons/SniperRifles/MilitarySniper/MilitarySniper_Reload_Impact_03}}
 {id=WavePlayer_8, properties {soundWaveAssetPtr=/Game/Sounds/Fort_Weapons/SniperRifles/MilitarySniper/MilitarySniper_Reload_Impact_04}}
]
properties {
 attenuationSettings=&Fort_Weapon_Reload_Att
 bDebug=true
 bOverrideConcurrency=true
 maxDistance=2250
 soundClassObject=&Fort_WeaponFoley
 volumeMultiplier=0.501
}
