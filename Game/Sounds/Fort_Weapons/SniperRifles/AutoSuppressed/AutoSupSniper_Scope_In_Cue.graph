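# Sound cue for scoping in with the suppressed automatic sniper rifle, played at low volume in the weapon foley sound class.
# A mixer layers a full-level scope-in sample over a quieter random pick from four player movement foley clips.
# Each layer passes through its own modulator for slight variation.
# Overridden attenuation keeps the sound short-range, and overridden concurrency prevents new instances once the limit is reached.
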SoundCue /Game/Sounds/Fort_Weapons/SniperRifles/AutoSuppressed/AutoSupSniper_Scope_In_Cue
root=Mixer_0
nodes [
 {id=Mixer_0, children [Modulator_0,Modulator_1], properties {inputVolume [0.4,1]}}
 {id=Modulator_0, children [Random_4]}
 {id=Modulator_1, children [WavePlayer_3], properties {pitchMax=1}}
 {id=Random_4, children [WavePlayer_19, WavePlayer_1565, WavePlayer_1563, WavePlayer_1562], properties {weights [1,1,1,1]}}
 {id=WavePlayer_1562, properties {soundWaveAssetPtr=/Game/Sounds/Fort_Foley_Sounds/Moves/player_run_foley_12}}
 {id=WavePlayer_1563, properties {soundWaveAssetPtr=/Game/Sounds/Fort_Foley_Sounds/Moves/player_run_foley_10}}
 {id=WavePlayer_1565, properties {soundWaveAssetPtr=/Game/Sounds/Fort_Foley_Sounds/Moves/player_run_foley_03}}
 {id=WavePlayer_19, properties {soundWaveAssetPtr=/Game/Sounds/Fort_Foley_Sounds/Moves/player_run_foley_09}}
 {id=WavePlayer_3, properties {soundWaveAssetPtr=&/AutoSupSniper_Scope_In_01}}
]
properties {
 bOverrideAttenuation=true
 bOverrideConcurrency=true
 maxDistance=2700
 soundClassObject=&Fort_WeaponFoley
 volumeMultiplier=0.15
 attenuationOverrides {falloffDistance=1500, nonSpatializedRadiusStart=200, attenuationShapeExtents {x=1200}}
 concurrencyOverrides {resolutionRule=PreventNew}
}
